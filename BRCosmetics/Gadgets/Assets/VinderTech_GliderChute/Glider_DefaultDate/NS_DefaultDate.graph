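# Niagara effect for the DefaultDate VinderTech glider, drawing jet meshes and ribbon streaks from the glider's skeletal mesh.
# Jets burst two mesh particles, pinned to sockets by spawn group and stretched further when a system integrator value passes a threshold.
# Streaks spawns two socket-attached ribbons, pushing particles along sampled X with age, farther as that integrator rises, while fading.
# User parameters set colors, alpha and ribbon size.
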NiagaraSystem /BRCosmetics/Gadgets/Assets/VinderTech_GliderChute/Glider_DefaultDate/NS_DefaultDate
dependencies {
 usedDataInterfaces [&NiagaraDataInterfaceCurve]
 usedDynamicInputs []
 usedModules [
  &SystemState
  &self:SystemSpawnScript.SystemScriptSource.SystemScriptGraph.NiagaraNodeAssignment_1.SetVariables_#1
  /Game/Effects/Niagara/Modules/Math/FN_Integrator/FN_Integrator_v2
 ]
}
emitters {
 Jets {
  emitterUpdateScript [
   &EmitterState(
    Inactive Response=::NewEnumerator2
    Life Cycle Mode=::NewEnumerator1
    Loop Behavior=::NewEnumerator1
    Loop Delay=0.5
    Loop Duration Mode=::NewEnumerator1
    UseLoopDelay=false
   )
   &SpawnBurst_Instantaneous(Spawn Count=2, Use Spawn Probability=false)
  ]
  particleSpawnScript [
   /Niagara/Modules/Spawn/Initialization/InitializeParticle(
    Lifetime=0.5
    Mesh Scale=<0.97 0.93 0.8>
    Module.Write Color=true
    Module.Write Lifetime=true
    Module.Write Mass=true
    Module.Write Material Random=true
    Module.Write Position=false
    Module.Write Scale=true
    Module.Write SpriteRotation=false
    Module.Write SpriteSize=true
    Position=&SimulationPosition()
    Sprite Size=<5 5>
   ) {enabled=false}
   /Niagara/Modules/Spawn/Initialization/V2/InitializeParticle(
    Lifetime=0.5
    Mesh Scale=<0.97 0.93 0.8>
    Mesh Scale Mode=::NewEnumerator4
    Sprite Size=<5 5>
    Sprite Size Mode=::NewEnumerator4
    UsePositionOffset=false
    Write Material Random=true
   ) {module=InitializeParticle001}
   &FN_SetSpawnGroup_v2(Number of Spawn Groups=2) {enabled=false}
   &FN_SetSpawnGroup(Number of Spawn Groups=2)
  ]
  particleUpdateScript [
   &ParticleState(Kill Particles When Lifetime Has Elapsed=false, Let Infinitely Lived Particles Die When Emitter Deactivates=true)
   &SkeletalMeshLocation(
    Apply Sampled Orientation to Mesh Particles=true
    Bone / Socket Index=$Particles.SpawnGroup
    Inherit Velocity=false
    Interpolate Mesh Sampling=false
    MaskSpawnGroup=false
    Mesh Sampling Type=::NewEnumerator5
    Sampled Position Offset=<-5 0 -0.3>
    Sampled Position Offset Coordinate Space:ENiagaraMeshTransforms=::NewEnumerator3
    Skeletal Mesh=$User.SkeletalMesh
    Socket Sampling Mode=::NewEnumerator5
    Use Sampled Mesh Attribute as Ribbon ID=false
   )
   &ScaleMeshSize(Scale Factor=&MakeVector(
     X=&Multiply_Float(
      A=&Multiply_Float(A=1, B=HlslExpression() {hlslExpression="System.FN_Integrator_v2.Attr > .15 ? 2.2 : 1.3"})
      B=&FloatFromCurve(FloatCurve=DataInterface() {propertyValues {Curve {keys [
           {interpMode=RCIM_Cubic, tangentMode=RCTM_Break, time=0, value=0}
           {arriveTangent=3.1967, interpMode=RCIM_Cubic, tangentMode=RCTM_Break, time=1, value=1}
          ]}}})
     )
     Y=1
     Z=1
    ))
   &Color(Color=$User.Jet Color, Scale Alpha=$User.Jet_Alpha, Scale Color=&VectorFromFloat(Value=10))
  ]
  rendererClasses [&NiagaraMeshRendererProperties]
  renderers [{rendererClass=&NiagaraMeshRendererProperties, rendererIndex=0}]
 }
 Streaks {
  emitterUpdateScript [
   &EmitterState()
   &FN_SpawnRibbon_v2(Number of Ribbons=2, SpawnRate=20)
  ]
  particleSpawnScript [
   &FN_InitializeRibbon_v3(
    Color:LinearColor=<rgba 1 0.15 1 1>
    Enable Skeletal Position=true
    Facing Axis:ENiagaraOrientationAxis=::ZAxis
    Interpolate Bone Sampling=false
    Lifetime=1.5
    Ribbon Width:NiagaraFloat=$User.Ribbon Width
   )
   /SetVariables(Particles.RibbonFacing=&RotateVectorByQuaternion(Vector To Rotate=<0 0 1>, Quaternion=$Particles.SampledRotation)) {isSetParameters=true}
  ]
  particleUpdateScript [
   &ParticleState()
   /SetVariables(Particles.Position=&Add_Vector(
     A=Linked() {linkedVariable {name=Particles.Initial.Position, type=NiagaraPosition}}
     B=&Multiply_VectorByFloat(
      Vector=&RotateVectorByQuaternion(
       Vector To Rotate=&MakeVector(
        X=&Multiply_Float(A=&Lerp_Float(A=150, B=250, Alpha=$System.FN_Integrator_v2.Attr), B=$User.Ribbon Length)
        Y=0
        Z=0
       )
       Quaternion=$Particles.SampledRotation
      )
      Float=$Particles.NormalizedAge
     )
    )) {isSetParameters=true}
   &SolveForcesAndVelocity(Clamp Velocity=false, Limit Acceleration=false)
   /Game/Effects/Niagara/Modules/Contrails/FN_UpdateRibbon/FN_UpdateRibbon_v3(Enable Skeletal Position=true)
   /SetVariables(Particles.RibbonFacing=&MakeCustomVectorFromBool(
     Bool=$Particles.Locked
     True Vector=&RotateVectorByQuaternion(Vector To Rotate=<0 0 1>, Quaternion=$Particles.FN_UpdateRibbon_v3.SampledRotation)
     False Vector=$Particles.RibbonFacing
    )) {isSetParameters=true}
   &ScaleColor(
    Scale RGB=&VectorFromFloat(Value=&FloatFromCurve(FloatCurve=DataInterface() {propertyValues {Curve {keys [<at 0 1 RCIM_Cubic>,<at 0.5 0.9 RCIM_Cubic>]}}}))
    ScaleA=true
    ScaleRGB=true
   )
   &Color(
    Color=$User.Ribbon Color
    Scale Alpha=&FloatFromCurve(
     FloatCurve=DataInterface() {propertyValues {Curve {keys [{interpMode=RCIM_Cubic, leaveTangent=0.2296, tangentMode=RCTM_Break, time=0, value=1}, <at 1 0>]}}}
     Scale Curve=&Multiply_Float(A=$User.Ribbon Alpha, B=1)
    )
   )
  ]
  rendererClasses [&NiagaraRibbonRendererProperties]
  renderers [{rendererClass=&NiagaraRibbonRendererProperties}]
 }
}
summary {
 systemName=NS_DefaultDate
 emitters [
  {emitterName=Jets, rendererClasses [&NiagaraMeshRendererProperties]}
  {emitterName=Streaks, rendererClasses [&NiagaraRibbonRendererProperties]}
 ]
 userVariables [
  User.bIsFrontEndPreview {type=NiagaraBool}
  User.bIsFullyDeployed {type=NiagaraBool}
  User.ForwardDot {type=NiagaraFloat, defaultValue=0, description=""}
  User.Jet Color {type=LinearColor, description="", defaultValue {struct=LinearColor, value=<rgba 1 1 1 1>}}
  User.Jet_Alpha {type=NiagaraFloat, defaultValue=1, description=""}
  User.PlayerSpeed {type=NiagaraFloat, defaultValue=0, description=""}
  User.Ribbon Alpha {type=NiagaraFloat, defaultValue=1, description=""}
  User.Ribbon Color {type=LinearColor, description="", defaultValue {struct=LinearColor, value=<rgba 1 1 1 1>}}
  User.Ribbon Length {type=NiagaraFloat, defaultValue=1, description=""}
  User.Ribbon Width {type=NiagaraFloat, defaultValue=50, description=""}
  User.RibbonIDOffset {type=NiagaraInt32}
  User.RightDot {type=NiagaraFloat, defaultValue=0, description=""}
  User.SkeletalMesh {
   type=NiagaraDataInterfaceSkeletalMesh
   defaultValue {
    struct=NiagaraExt_VariableValue_DataInterface
    value {dataInterface=&self:NiagaraDataInterfaceSkeletalMesh_0}
   }
  }
 ]
}
